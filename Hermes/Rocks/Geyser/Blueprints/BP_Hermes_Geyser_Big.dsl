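; Big geyser actor that periodically erupts to launch players, vehicles, projectiles, C4, toys and physics objects.
; An authority timer loop cycles through random wait, prime and fire, with OnRep functions toggling cone collision, Niagara effects and gameplay cues.
; Cone overlap applies type-specific launches, impulses or gravity-scale changes, and exit restores gravity and negates fall damage.
; Pawn and vehicle launches can destroy props along their path, and pawn launches report a quest stat.
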(event Custom|ActiveGeyser)

(event Custom|DeactivateGeyser)

(event EventBeginPlay
  ;[0.05098 1 0.01223] Begin Play
  (Variables|Default|SetisDisabled true)
  (Utilities|FlowControl|Sequence
    (:then_0
      (CallFunction|SetupMaterialParams)
      (CallFunction|SetupGeyserWait)
      (Variables|Default|SetSavedConeScale (Variables|/Script/Engine.SceneComponent|GetRelativeScale3D (Variables|Default|GetS_FX_GeyserCone)))
      (Audio|PlaySoundatLocation "/Game/Sounds/Artemis/Environment_Actors/RealityForest_Geyser/RealityForest_Daisy_Geyser_Amb_Cue.RealityForest_Daisy_Geyser_Amb_Cue" (Transformation|GetWorldLocation (Variables|Default|GetLaunchSound))))))

(event Custom|SetupMaterialParams
  (bind _s_fx_geysercone (Variables|Default|GetS_FX_GeyserCone))
  ;[0.008445 0.4764 1] Set material values
  (Rendering|Material|SetScalarParameterValueonMaterials _s_fx_geysercone "UVScale" (Math|Random|RandomFloatinRange 2.0 4.0))
  (Rendering|Material|SetScalarParameterValueonMaterials _s_fx_geysercone "SlantAmount" (Math|Random|RandomFloatinRange -5e+02 5e+02))
  (Rendering|Material|SetScalarParameterValueonMaterials _s_fx_geysercone "WPO_Amount" (Math|Random|RandomFloatinRange 0.0 5e+02))
  (Rendering|Material|SetScalarParameterValueonMaterials _s_fx_geysercone "WorldTexUVScale" (Math|Random|RandomFloatinRange 0.5 2.0))
  (bind _output_get (Variables|Default|SetGeyserContrast (Math|Random|RandomFloatinRange 1.0 3.0)))
  (Rendering|Material|SetScalarParameterValueonMaterials _s_fx_geysercone "Top Edge Contrast" _output_get))

(event Custom|SetupGeyserWait
  ;[1 0 0.9359] Timed Firing. Setup the "wait" time before the geyser fires. Prime the geyser to allow players to enter/exit. Fire to launch the players. Reset the flow.
  (Networking|SwitchHasAuthority
    (:Authority
      (Utilities|Time|SetTimerbyEvent (delegate Custom|PrimeGeyser) (Math|Random|RandomFloatinRange (ScalableFloat|GetValueatLevel (Variables|Default|GetSF_MinActivationTime)) (ScalableFloat|GetValueatLevel (Variables|Default|GetSF_MaxActivationTime)))))))

(event Custom|PrimeGeyser
  ;[1 0 0.9359] Timed Firing. Setup the "wait" time before the geyser fires. Prime the geyser to allow players to enter/exit. Fire to launch the players. Reset the flow.
  (Variables|Default|SetPrimeMode true)
  (Utilities|Time|SetTimerbyEvent (delegate Custom|FireGeyser) (ScalableFloat|GetValueatLevel (Variables|Default|GetSF_PrimeDuration))))

(event Custom|FireGeyser
  ;[1 0 0.9359] Timed Firing. Setup the "wait" time before the geyser fires. Prime the geyser to allow players to enter/exit. Fire to launch the players. Reset the flow.
  (Variables|Default|SetPrimeMode false)
  (Variables|Default|SetisDisabled false)
  (Utilities|Time|SetTimerbyEvent (delegate Custom|ResetGeyserWait) (ScalableFloat|GetValueatLevel (Variables|Default|GetSF_ActiveDuration))))

(event OnComponentBeginOverlap(S_FX_GeyserCone)
  (params (OverlappedComponent :type Primitive%20Component%20Object%20Reference) (OtherActor :type Actor%20Object%20Reference) (OtherComp :type Primitive%20Component%20Object%20Reference) (OtherBodyIndex :type Integer) (bFromSweep :type Boolean) (SweepResult :type Hit%20Result%20Structure%20%28by%20ref%29))
  (bind _sf_launchvelocity (Variables|Default|GetSF_LaunchVelocity))
  (bind _returnvalue (ScalableFloat|GetValueatLevel _sf_launchvelocity))
  (bind _returnvalue_2 (Transformation|GetActorUpVector))
  (bind _returnvalue_3 (* _returnvalue_2 _returnvalue))
  (bind _returnvalue_4 (Pawn|GetMovementBaseActor 0))
  (bind _returnvalue_5 (Transformation|GetActorUpVector))
  (bind _sf_vehiclelaunchvelocity (Variables|Default|GetSF_VehicleLaunchVelocity))
  (bind _returnvalue_6 (ScalableFloat|GetValueatLevel _sf_vehiclelaunchvelocity))
  (bind _returnvalue_7 (* _returnvalue_5 _returnvalue_6))
  (bind (_returnvalue_x _returnvalue_y _returnvalue_z) (Math|Vector|BreakVector (Transformation|GetActorLocation)))
  ;[1 0.3378 0] Launch players and vehicles if they overlap the geyser while it's active. Handle projectiles and toys/C4 separately.
  (if (Variables|Default|GetShouldBoost)
    (bind _asfort_pawn (Utilities|Casting|CastToFortPawn OtherActor)
      (:then
        (Utilities|IsValid _returnvalue_4
          (:Is%20Valid
            (if (Utilities|IsValid (Actor|GetComponentbyClass _returnvalue_4 "/Script/FortniteGame.FortPhysicsObjectComponent") :!class /Script/Engine.KismetSystemLibrary)
              (bind _output_get (Variables|Default|SetPhysicsObjectToLaunch _returnvalue_4))
              (Variables|Default|SetLaunchedActor _output_get)))
          (:Is%20Not%20Valid
            (if (not (Math|Boolean|ORBoolean (GameplayTags|TagContainer|CompareTagAssetInterfacetoOtherTagContainers (Variables|Default|GetHawaii) (Variables|Default|GetRiding) (Variables|Default|GetKnockbackImmunity) _asfort_pawn) (.Case_1%20True (GameplayTags|TagContainer|CompareTagAssetInterfacetoOtherTagContainers (Variables|Default|GetHawaii) (Variables|Default|GetRiding) (Variables|Default|GetKnockbackImmunity) _asfort_pawn)) (.Case_2%20True (GameplayTags|TagContainer|CompareTagAssetInterfacetoOtherTagContainers (Variables|Default|GetHawaii) (Variables|Default|GetRiding) (Variables|Default|GetKnockbackImmunity) _asfort_pawn))))
              (Networking|SwitchHasAuthority
                (:Authority
                  (Variables|Default|SetLaunchedActor _asfort_pawn)
                  (Ability|SendGameplayEventtoActor _asfort_pawn "(TagName=\"Abilities.Generic.Vent.Enter\")" (Utilities|Struct|MakeGameplayEventData :Target self :ContextHandle "()" :InstigatorTags "(GameplayTags=)" :TargetTags "(GameplayTags=)" :EventMagnitude (Variables|Default|GetUpwardLaunchVelocity) :TargetData "()"))
                  (Character|LaunchCharacter _asfort_pawn _returnvalue_3 (Variables|Default|GetXYOverride) (Variables|Default|GetZOverride))
                  ;;[0.6434 0.2072 1] Use Geysers in Different Matches Quest hook in
                  (Utilities|IsValid (Fortnite|GetFortPlayerControllerfromActor _asfort_pawn)
                    (:Is%20Valid
                      (Utilities|IsValid (Quests|GetQuestManager (Fortnite|GetFortPlayerControllerfromActor _asfort_pawn) "Invalid")
                        (:Is%20Valid
                          (Quest|SendComplexCustomStatEvent (Quests|GetQuestManager (Fortnite|GetFortPlayerControllerfromActor _asfort_pawn) "Invalid") 0 "(GameplayTags=((TagName=\"Athena.Quests.Guyser\")))" "(GameplayTags=)" 1))))
                    (:Is%20Not%20Valid
                      _))
                  (CallFunction|DestroyBuildingProps _asfort_pawn (Ability|GetAbilitySystemComponent _asfort_pawn) _returnvalue_3))
                (:Remote
                  (bind (_returnvalue_x_2 _returnvalue_y_2 _returnvalue_z_2) (Math|Vector|BreakVector (Transformation|GetActorLocation _asfort_pawn)))
                  (Projectile|AddGameplayCue%28Looping%29Local _asfort_pawn (Variables|Default|GetVentCue) (Ability|GameplayCue|MakeGameplayCueParameters :NormalizedMagnitude (Math|Conversions|ToFloat(Boolean) (<= _returnvalue_z_2 (+ _returnvalue_z 8343.651367))) :RawMagnitude 0.0 :EffectContext "()" :AggregatedSourceTags "(GameplayTags=)" :AggregatedTargetTags "(GameplayTags=)" :Location "0,0,0" :Normal "0,0,0" :GameplayEffectLevel 1 :AbilityLevel 1 :bReplicateLocationWhenUsingMinimalRepProxy false))
                  (ForceFeedback|SpawnForceFeedbackAttached "/Game/Weapons/ForceFeedback/Explosions/FF_Generic_ExplosionLarge_Near.FF_Generic_ExplosionLarge_Near" (Variables|/Script/Engine.Actor|GetRootComponent _asfort_pawn))))))) ; Do not launch players if they are riding on an object we will launch
      )
      (:CastFailed
        (bind _asfort_vehicle_interface (Utilities|Casting|CastToFortVehicleInterface OtherActor)
          (:then
            (CallFunction|OnVehicleEnterZone (Vehicle|GetActor _asfort_vehicle_interface) -0.6)
            (Vehicle|SetIgnoreAllFallingDamage _asfort_vehicle_interface true)
            (Variables|Default|SetLaunchedActor (Vehicle|GetActor _asfort_vehicle_interface))
            (Audio|BroadcastAbilityImpactatLocation 0 (Vehicle|GetActor _asfort_vehicle_interface) (+ (Vehicle|GetVehicleVelocity _asfort_vehicle_interface) _returnvalue_7) 777.0)
            (CallFunction|DestroyBuildingProps (Vehicle|GetActor _asfort_vehicle_interface) (Ability|GetAbilitySystemComponent (Vehicle|GetActor _asfort_vehicle_interface)) (+ (Vehicle|GetVehicleVelocity _asfort_vehicle_interface) _returnvalue_7)))
          (:CastFailed
            (bind _asfort_projectile_base (Utilities|Casting|CastToFortProjectileBase OtherActor)
              (:then
                (Variables|Default|SetLaunchedActor _asfort_projectile_base)
                (CallFunction|ProjEnter _asfort_projectile_base))
              (:CastFailed
                (bind _asbuilding_gameplay_actor_c4 (Utilities|Casting|CastToBuildingGameplayActorC4 OtherActor)
                  (:then
                    (bind _returnvalue_8 (Actor|GetComponentbyClass _asbuilding_gameplay_actor_c4 "/Script/Engine.ProjectileMovementComponent"))
                    (Variables|/Script/Engine.ProjectileMovementComponent|SetProjectileGravityScale (* (Variables|/Script/Engine.ProjectileMovementComponent|GetProjectileGravityScale _returnvalue_8) -1.0) _returnvalue_8)
                    (Variables|Default|SetLaunchedActor _asbuilding_gameplay_actor_c4))
                  (:CastFailed
                    (bind _asfort_physics_ball (Utilities|Casting|CastToFortPhysicsBall OtherActor)
                      (:then
                        (Variables|Default|SetLaunchedActor _asfort_physics_ball)
                        (CallFunction|ToyEnter _asfort_physics_ball))
                      (:CastFailed
                        (if (Utilities|IsValid (Actor|GetComponentbyClass OtherActor "/Script/FortniteGame.FortPhysicsObjectComponent") :!class /Script/Engine.KismetSystemLibrary)
                          (Variables|Default|SetLaunchedActor OtherActor)
                          (Variables|Default|SetPhysicsObjectToLaunch OtherActor))))))))))))
    (FortPhysicsSimulationLibrary|AddPhysicsImpulse (Variables|Default|GetPhysicsObjectToLaunch) (* (Transformation|GetActorUpVector) (ScalableFloat|GetValueatLevel (Variables|Default|GetSF_BoulderLaunchVelocity))) true "None" false false 0 "(MaxSimSizeAllowed=Large,bRejectEmbeddedInLandscape=False)")))

(event Custom|ResetGeyserWait
  ;[1 0 0.9359] Timed Firing. Setup the "wait" time before the geyser fires. Prime the geyser to allow players to enter/exit. Fire to launch the players. Reset the flow.
  (Variables|Default|SetisDisabled true)
  (CallFunction|SetupGeyserWait))

(event OnComponentEndOverlap(S_FX_GeyserCone)
  (params (OverlappedComponent :type Primitive%20Component%20Object%20Reference) (OtherActor :type Actor%20Object%20Reference) (OtherComp :type Primitive%20Component%20Object%20Reference) (OtherBodyIndex :type Integer))
  ;[0.1134 0 1] Set the launched targets to be immune to fall damage.
  (if (Variables|Default|GetShouldBoost)
    (bind _asfort_pawn (Utilities|Casting|CastToFortPawn OtherActor)
      (:then
        (Ability|SendGameplayEventtoActor _asfort_pawn "(TagName=\"Abilities.Generic.Vent.Exit\")")
        (Variables|/Script/FortniteGame.FortPawn|SetbIgnoreNextFallingDamage true _asfort_pawn)
        (Projectile|RemoveGameplayCue%28Looping%29Local _asfort_pawn (Variables|Default|GetVentCue) (Ability|GameplayCue|MakeGameplayCueParameters :NormalizedMagnitude 0.0 :RawMagnitude 0.0 :EffectContext "()" :AggregatedSourceTags "(GameplayTags=)" :AggregatedTargetTags "(GameplayTags=)" :Location "0,0,0" :Normal "0,0,0" :GameplayEffectLevel 1 :AbilityLevel 1 :bReplicateLocationWhenUsingMinimalRepProxy false)))
      (:CastFailed
        (bind _asfort_vehicle_interface (Utilities|Casting|CastToFortVehicleInterface OtherActor)
          (:then
            (CallFunction|OnVehicleExitZone (Vehicle|GetActor _asfort_vehicle_interface))
            (Vehicle|SetIgnoreAllFallingDamage _asfort_vehicle_interface false)
            (Vehicle|SetIgnoreNextFallingDamage _asfort_vehicle_interface true))
          (:CastFailed
            (bind _asfort_projectile_base (Utilities|Casting|CastToFortProjectileBase OtherActor)
              (:then
                (CallFunction|ProjExit _asfort_projectile_base))
              (:CastFailed
                (bind _asbuilding_gameplay_actor_c4 (Utilities|Casting|CastToBuildingGameplayActorC4 OtherActor)
                  (:then
                    (bind _returnvalue (Actor|GetComponentbyClass _asbuilding_gameplay_actor_c4 "/Script/Engine.ProjectileMovementComponent"))
                    (Variables|/Script/Engine.ProjectileMovementComponent|SetProjectileGravityScale (* (Variables|/Script/Engine.ProjectileMovementComponent|GetProjectileGravityScale _returnvalue) -1.0) _returnvalue))
                  (:CastFailed
                    (bind _asfort_physics_ball (Utilities|Casting|CastToFortPhysicsBall OtherActor))
                    (CallFunction|ToyExit _asfort_physics_ball)
                    (if (Utilities|IsValid (Actor|GetComponentbyClass _asfort_physics_ball "/Script/FortniteGame.FortPhysicsObjectComponent") :!class /Script/Engine.KismetSystemLibrary)
                      (PhysicsObject|DecrementNeverSleep (Actor|GetComponentbyClass _asfort_physics_ball "/Script/FortniteGame.FortPhysicsObjectComponent")))))))))))))

(fn DestroyBuildingProps
  (params (Actor :type Actor%20Object%20Reference) (AbilitySystemComp :type Ability%20System%20Component%20Object%20Reference) (LaunchVelocity :type Vector))
  (if (> (ScalableFloat|GetValueatLevel (Variables|Default|GetShouldDestroy)) 0.0)
    (bind _outhits (Collision|MultiCapsuleTraceForObjects :Start (Transformation|GetActorLocation Actor) :End (+ (Transformation|GetActorLocation Actor) (* (Math|Vector|Normalize LaunchVelocity) (ScalableFloat|GetValueatLevel (Variables|Default|GetDestroyDistance)))) :Radius 8e+01 :HalfHeight 1.5e+02 :ObjectTypes (Variables|Default|GetDestroyObjectTypes) :bTraceComplex false :DrawDebugType "None" :bIgnoreSelf true :TraceColor "(R=1.000000,G=0.000000,B=0.000000,A=1.000000)" :TraceHitColor "(R=0.000000,G=1.000000,B=0.000000,A=1.000000)" :DrawTime 8.0))
    (for _array_element _outhits
      (if (Variables|/Script/Engine.Actor|GetbCanBeDamaged (.HitActor (Collision|BreakHitResult _array_element)))
        (bind _bdidactorhavetags (Class|AthenaQuestFunctionLibrary|AthenaCheckActorHasTags (.HitActor (Collision|BreakHitResult _array_element)) "(GameplayTags=((TagName=\"Deimos.Spawner\")))"))
        (if (not _bdidactorhavetags)
          (Utilities|Casting|CastToBuildingTrap (.HitActor (Collision|BreakHitResult _array_element))
            (:CastFailed
              (GameplayEffects|ApplyGameplayEffectToTarget :self AbilitySystemComp :GameplayEffectClass "/Game/Athena/Items/Consumables/ShockwaveGrenade/GE_ShockGrenade_Damage.GE_ShockGrenade_Damage_C" :Target (Ability|GetAbilitySystemComponent (.HitActor (Collision|BreakHitResult _array_element))) :Level 0.0))))))))

(fn GetCollisionStaticMesh
  (return "0, 0, 0" "/Game/Athena/Environments/Island/Runes/Mesh/SM_CorruptedVent_A.SM_CorruptedVent_A"))

(fn GetMeshComponents
  (return 0))

(fn IgnoreWorldGridScaleOverride
  (return false))

(fn OnRep_IsActive
  (Networking|SwitchHasAuthority
    (:Remote
      (if (Variables|Default|GetIsActive)
        (CallFunction|ActiveGeyser)
        (else
          (CallFunction|DeactivateGeyser))))))

(fn OnRep_isDisabled
  (if (Variables|Default|GetisDisabled)
    (Collision|SetCollisionEnabled (Variables|Default|GetS_FX_GeyserCone) "NoCollision")
    (else
      (Collision|SetCollisionEnabled (Variables|Default|GetS_FX_GeyserCone) "QueryAndPhysics")
      (if (Variables|Default|GetAllowEffects)
        (bind _returnvalue (Effects|Components|ParticleSystem|SpawnFXAtLocation "/Game/Effects/Fort_Effects/Gameplay/Geyser/NS_Geyser_Burst.NS_Geyser_Burst" (Transformation|GetWorldLocation (Variables|Default|GetNS_Geyser_Idle))))
        (Variables|Default|SetGeyserBurstRef _returnvalue)
        (GameplayCue|ExecuteGameplayCueOnActor(Burst) self (Variables|Default|GetLaunchGC) (Ability|GameplayCue|MakeGameplayCueParameters :NormalizedMagnitude 0.0 :RawMagnitude 0.0 :Location "0,0,0" :Normal "0,0,0" :Instigator (Variables|Default|GetLaunchedActor) :GameplayEffectLevel 1 :AbilityLevel 1 :bReplicateLocationWhenUsingMinimalRepProxy false))
        (Variables|Default|SetLaunchedActor 0) ; Clearing so the next time an ambient launch comes around we don't have this cached
      ))))

(fn OnRep_PrimeMode
  (bind _ns_geyser_idle (Variables|Default|GetNS_Geyser_Idle))
  (if (Variables|Default|GetPrimeMode)
    (Effects|Components|ParticleSystem|SetBoolParameter _ns_geyser_idle "User.IsPrimed" true)
    (GameplayCue|ExecuteGameplayCueOnActor(Burst) self (Variables|Default|GetGeyserPrime) (Ability|GameplayCue|MakeGameplayCueParameters 0.0))
    (else
      (Effects|Components|ParticleSystem|SetBoolParameter _ns_geyser_idle "User.IsPrimed" false))))

(fn ProjEnter
  (params (Projectile :type Fort%20Projectile%20Base%20Object%20Reference))
  (bind _projectilemovementcomponent (Variables|/Script/FortniteGame.FortProjectileBase|GetProjectileMovementComponent Projectile))
  (bind _projectilegravityscale (Variables|/Script/Engine.ProjectileMovementComponent|GetProjectileGravityScale _projectilemovementcomponent))
  (if (> _projectilegravityscale 0.0)
    (Variables|/Script/Engine.ProjectileMovementComponent|SetProjectileGravityScale (* _projectilegravityscale -1.0) _projectilemovementcomponent)
    (else
      (Variables|/Script/Engine.ProjectileMovementComponent|SetProjectileGravityScale (+ _projectilegravityscale (Variables|Default|GetZeroGravPrj)) _projectilemovementcomponent))))

(fn ProjExit
  (params (Projectile :type Fort%20Projectile%20Base%20Object%20Reference))
  (bind _projectilemovementcomponent (Variables|/Script/FortniteGame.FortProjectileBase|GetProjectileMovementComponent Projectile))
  (bind _projectilegravityscale (Variables|/Script/Engine.ProjectileMovementComponent|GetProjectileGravityScale _projectilemovementcomponent))
  (if (== _projectilegravityscale (Variables|Default|GetZeroGravPrj))
    (Variables|/Script/Engine.ProjectileMovementComponent|SetProjectileGravityScale (- _projectilegravityscale (Variables|Default|GetZeroGravPrj)) _projectilemovementcomponent)
    (else
      (Variables|/Script/Engine.ProjectileMovementComponent|SetProjectileGravityScale (* _projectilegravityscale -1.0) _projectilemovementcomponent))))

(fn ToyEnter
  (params (Projectile :type Fort%20Physics%20Ball%20Object%20Reference))
  (bind _returnvalue (Actor|GetComponentbyClass Projectile "/Script/Engine.ProjectileMovementComponent"))
  (if (> (Variables|/Script/Engine.ProjectileMovementComponent|GetProjectileGravityScale _returnvalue) 0.0)
    (Variables|/Script/Engine.ProjectileMovementComponent|SetProjectileGravityScale (* (Variables|/Script/Engine.ProjectileMovementComponent|GetProjectileGravityScale _returnvalue) -0.5) _returnvalue)
    (else
      (Variables|/Script/Engine.ProjectileMovementComponent|SetProjectileGravityScale (+ (Variables|/Script/Engine.ProjectileMovementComponent|GetProjectileGravityScale _returnvalue) (Variables|Default|GetZeroGravPrj)) _returnvalue))))

(fn ToyExit
  (params (Projectile :type Fort%20Physics%20Ball%20Object%20Reference))
  (bind _returnvalue (Actor|GetComponentbyClass Projectile "/Script/Engine.ProjectileMovementComponent"))
  (if (== (Variables|/Script/Engine.ProjectileMovementComponent|GetProjectileGravityScale _returnvalue) (Variables|Default|GetZeroGravPrj))
    (Variables|/Script/Engine.ProjectileMovementComponent|SetProjectileGravityScale (- (Variables|/Script/Engine.ProjectileMovementComponent|GetProjectileGravityScale _returnvalue) (Variables|Default|GetZeroGravPrj)) _returnvalue)
    (else
      (Variables|/Script/Engine.ProjectileMovementComponent|SetProjectileGravityScale (* (Variables|/Script/Engine.ProjectileMovementComponent|GetProjectileGravityScale _returnvalue) -2.0) _returnvalue))))

(fn ConstructionScript
  (|CallParentFunction))
